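; Pawn Blueprint for the Nug, a chicken prey NPC that players can pick up, carry and throw.
; Pickup exchanges gameplay effects with the holder, attaches via the held object component and pauses storm checks.
; Replicated held/thrown flags drive collision, movement and animation, and bind holder bump events that trigger a melee ability.
; Throws send first-time quest events, wind splines cut flight short, and landing restores pickup after delays.
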(event EventOnDamageServer
  (params (Damage :type Float%20%28single-precision%29) (DamageTags :type Gameplay%20Tag%20Container%20Structure%20%28by%20ref%29) (Momentum :type Vector :default "0, 0, 0") (HitInfo :type Hit%20Result%20Structure%20%28by%20ref%29) (InstigatedBy :type Controller%20Object%20Reference) (DamageCauser :type Actor%20Object%20Reference) (EffectContext :type Gameplay%20Effect%20Context%20Handle%20Structure))
  (bind _ge_recentlyattackedbypawn (Variables|Default|GetGE_RecentlyAttackedByPawn))
  ;[1 1 1] Track recent Damage from Pawn sources
  (|CallParentFunction Damage DamageTags Momentum HitInfo InstigatedBy DamageCauser EffectContext)
  (if (Utilities|IsValidClass _ge_recentlyattackedbypawn)
    (Utilities|IsValid (Pawn|GetControlledPawn InstigatedBy)
      (:Is%20Valid
        (GameplayEffects|ApplyGameplayEffectToSelf :self (Variables|GameplayAbilities|GetAIPawnAbilitySystemComponent) :GameplayEffectClass _ge_recentlyattackedbypawn :Level 0.0) ; Track that a Pawn has Damaged the Chicken recently.
      )) ; Was our last Damage Instigator a Pawn?
  ))

(event EventBeginPlay
  ;[1 1 1] Begin Play
  (|CallParentFunction)
  (CallFunction|UpdateMovementComponents)
  (EventDispatchers|BindEventtoOnPawnLanded (EventDispatchers|CreateEvent :!event Custom|NugLanded))
  (EventDispatchers|BindEventtoOnDied (delegate Custom|On%20Nug%20Died%20Event)))

(event Custom|On%20Nug%20Died%20Event
  (params (DamagedActor :type Actor%20Object%20Reference) (Damage :type Float%20%28single-precision%29) (InstigatedBy :type Controller%20Object%20Reference) (DamageCauser :type Actor%20Object%20Reference) (HitLocation :type Vector) (FHitComponent :type Primitive%20Component%20Object%20Reference) (BoneName :type Name) (Momentum :type Vector))
  (bind _thrown_handle (Variables|Default|GetThrown%20Handle))
  (bind _trystopguildeprojectiletimer (Variables|Default|GetTryStopGuildeProjectileTimer))
  ;[1 1 1] Nug On Death
  (CallFunction|UnassignPlayerBumpEvent) ; If held chicken dies, remove GE from player applied when holding a chicken
  (Variables|Default|SetFortPlayerPawnHoldingChicken 0)
  (if (Utilities|Time|IsValidTimerHandle _thrown_handle)
    (Utilities|Time|ClearandInvalidateTimerbyHandle _thrown_handle)
    (else
      _))
  (if (Utilities|Time|IsValidTimerHandle _trystopguildeprojectiletimer)
    (Utilities|Time|ClearandInvalidateTimerbyHandle _trystopguildeprojectiletimer)))

(event Custom|AssignPlayerBumpEvent
  ;[1 1 1] Track Player who is holding Chicken
  ;
  ; If Player Bumps something, try to do the Bump Melee Attack.
  (if (ScalableFloat|GetValueasBool (Variables|Default|GetChickenTrackHeldBumpsHF))
    (Networking|SwitchHasAuthority
      (:Authority
        (bind _fortplayerpawnholdingchicken (Variables|Default|GetFortPlayerPawnHoldingChicken)
          (:then
            (Collision|AssignOnActorBeginOverlap _fortplayerpawnholdingchicken (delegate Custom|OnActorBeginOverlap_Event))
            (Collision|AssignOnActorHit (Variables|Default|GetFortPlayerPawnHoldingChicken) (delegate Custom|OnActorHit_Event))
            (EventDispatchers|AssignOnDied (Variables|Default|GetFortPlayerPawnHoldingChicken) (delegate Custom|OnDied_Event))
            (Game|AssignOnEndPlay (Variables|Default|GetFortPlayerPawnHoldingChicken) (delegate Custom|OnEndPlay_Event)))
          (:else))))))

(event EventOnNPCPawnInteractedWith
  (params (AIPawn :type Actor%20Object%20Reference) (InteractingPlayerPawn :type Fort%20Player%20Pawn%20Object%20Reference) (InteractionBeingAttempted :type EInteractionBeingAttempted%20Enum :default "FirstInteraction"))
  (bind _gameplayeffectcontexthandle (Utilities|Struct|MakeGameplayEffectContextHandle))
  (bind _aipawnabilitysystemcomponent (Variables|GameplayAbilities|GetAIPawnAbilitySystemComponent))
  (bind _thrown_handle (Variables|Default|GetThrown%20Handle))
  ;[1 1 1] InteractedWith->Give Chicken Stun ability, remove collision and set as held object
  (if (not (Variables|Default|GetbPawnPickedUp))
    (Variables|Default|SetFortPlayerPawnHoldingChicken InteractingPlayerPawn)
    (bind _fortplayerpawnholdingchicken (Variables|Default|GetFortPlayerPawnHoldingChicken)
      (:then
        (bind _returnvalue (GameplayEffects|ApplyGameplayEffectToTarget (Ability|GetFortAbilitySystemComponent _fortplayerpawnholdingchicken) (Variables|Default|GetGE_ApplyPickupAbility_Chicken) _aipawnabilitysystemcomponent 0.0 _gameplayeffectcontexthandle)) ; using this method to more easily get a player reference
        (Variables|Default|SetH_GE_PickedUp_Chicken _returnvalue)
        (bind _returnvalue_2 (GameplayEffects|ApplyGameplayEffectToTarget _aipawnabilitysystemcomponent (Variables|Default|GetGE_ApplyAbilities_Player) (Ability|GetFortAbilitySystemComponent _fortplayerpawnholdingchicken) 0.0 _gameplayeffectcontexthandle)) ; Inverse Targets on this so Chicken grants to Player
        (Variables|Default|SetH_GE_PickedUp_Player _returnvalue_2)
        (Variables|Default|SetbPawnPickedUp true) ; Toggle collison
        (CallFunction|UpdateMovementComponents)
        (CallFunction|NPCNugUpdateStormHandling)
        (CallFunction|PlayPickUpAudio)
        (bind _fortplayerpawnholdingchicken_2 (Variables|Default|GetFortPlayerPawnHoldingChicken)
          (:then
            (FortPawn|SetPawnandControlRotation (Math|Conversions|RotationFromXVector (Transformation|GetActorForwardVector _fortplayerpawnholdingchicken_2)))
            (Networking|ForceNetUpdate)
            (if (Networking|IsServer)
              (bind _fortplayerpawnholdingchicken_3 (Variables|Default|GetFortPlayerPawnHoldingChicken)
                (:then
                  (HeldObjects|Behavior|PickupHeldObject (Variables|Default|GetFortHeldObject) _fortplayerpawnholdingchicken_3)
                  (if (Utilities|Time|IsValidTimerHandle _thrown_handle)
                    (Utilities|Time|ClearandInvalidateTimerbyHandle _thrown_handle)))
                (:else))))
          (:else)))
      (:else))))

(event Custom|UnassignPlayerBumpEvent
  ;[1 1 1] Track Player who is holding Chicken
  ;
  ; If Player Bumps something, try to do the Bump Melee Attack.
  (if (ScalableFloat|GetValueasBool (Variables|Default|GetChickenTrackHeldBumpsHF))
    (Networking|SwitchHasAuthority
      (:Authority
        (bind _fortplayerpawnholdingchicken (Variables|Default|GetFortPlayerPawnHoldingChicken)
          (:then
            (Collision|UnbindEventfromOnActorBeginOverlap _fortplayerpawnholdingchicken (delegate Custom|OnActorBeginOverlap_Event))
            (Collision|UnbindEventfromOnActorHit (Variables|Default|GetFortPlayerPawnHoldingChicken) (delegate Custom|OnActorHit_Event))
            (EventDispatchers|UnbindEventfromOnDied (Variables|Default|GetFortPlayerPawnHoldingChicken) (delegate Custom|OnDied_Event))
            (Game|UnbindEventfromOnEndPlay (Variables|Default|GetFortPlayerPawnHoldingChicken) (delegate Custom|OnEndPlay_Event))
            (bind _fortplayerpawnholdingchicken_2 (Variables|Default|GetFortPlayerPawnHoldingChicken)
              (:then
                (Utilities|IsValid (Ability|GetFortAbilitySystemComponent _fortplayerpawnholdingchicken_2)
                  (:Is%20Valid
                    (GameplayEffects|RemoveActiveGameplayEffect (Ability|GetFortAbilitySystemComponent _fortplayerpawnholdingchicken_2) (Variables|Default|GetH_GE_PickedUp_Player) -1) ; Remove peck attack damage GE/Ability from Player
                  )))
              (:else)))
          (:else))))))

(event Custom|OnEndPlay_Event
  (params (Actor :type Actor%20Object%20Reference) (EndPlayReason :type EEndPlayReason%20Enum :default "Destroyed"))
  (bind _thrown_handle (Variables|Default|GetThrown%20Handle))
  ;[1 1 1] Track Player who is holding Chicken
  ;
  ; If Player Bumps something, try to do the Bump Melee Attack.
  (CallFunction|UnassignPlayerBumpEvent) ; If Player Dies or EndsPlay, stop tracking this Player
  (Variables|Default|SetFortPlayerPawnHoldingChicken 0) ; Clear this variable since the Chicken is no longer held.
  (if (Utilities|Time|IsValidTimerHandle _thrown_handle)
    (Utilities|Time|ClearandInvalidateTimerbyHandle _thrown_handle)))

(event Custom|OnDied_Event
  (params (DamagedActor :type Actor%20Object%20Reference) (Damage :type Float%20%28single-precision%29) (InstigatedBy :type Controller%20Object%20Reference) (DamageCauser :type Actor%20Object%20Reference) (HitLocation :type Vector) (FHitComponent :type Primitive%20Component%20Object%20Reference) (BoneName :type Name) (Momentum :type Vector))
  (bind _thrown_handle (Variables|Default|GetThrown%20Handle))
  ;[1 1 1] Track Player who is holding Chicken
  ;
  ; If Player Bumps something, try to do the Bump Melee Attack.
  (CallFunction|UnassignPlayerBumpEvent) ; If Player Dies or EndsPlay, stop tracking this Player
  (Variables|Default|SetFortPlayerPawnHoldingChicken 0) ; Clear this variable since the Chicken is no longer held.
  (if (Utilities|Time|IsValidTimerHandle _thrown_handle)
    (Utilities|Time|ClearandInvalidateTimerbyHandle _thrown_handle)))

(event Custom|OnActorBeginOverlap_Event
  (params (OverlappedActor :type Actor%20Object%20Reference) (OtherActor :type Actor%20Object%20Reference))
  ;[1 1 1] Track Player who is holding Chicken
  ;
  ; If Player Bumps something, try to do the Bump Melee Attack.
  (bind _data (Utilities|FlowControl|MergeExecutionwithData OtherActor (stale Custom|OnActorHit_Event OtherActor))
    (:Then
      (bind _asfort_pawn (Utilities|Casting|CastToFortPawn _data)) ; Thing we Bumped or Overlapped is a Pawn?
      (bind _fortplayerpawnholdingchicken (Variables|Default|GetFortPlayerPawnHoldingChicken)
        (:then
          (switch Utilities|FlowControl|Switch|SwitchonEFortTeamAffiliation (Team|GetTeamAffiliation _fortplayerpawnholdingchicken _asfort_pawn)
            (:Neutral
              _)
            (:Hostile
              _))
          (switch Utilities|FlowControl|Switch|SwitchonEFortTeamAffiliation (Team|GetTeamAffiliation self _asfort_pawn)
            (:Neutral
              _)
            (:Hostile
              _))
          (CallFunction|AttemptTriggerAbilityAfterPlayerBumpEvent))
        (:else)))))

(event Custom|OnActorHit_Event
  (params (SelfActor :type Actor%20Object%20Reference) (OtherActor :type Actor%20Object%20Reference) (NormalImpulse :type Vector) (Hit :type Hit%20Result%20Structure%20%28by%20ref%29))
  ;[1 1 1] Track Player who is holding Chicken
  ;
  ; If Player Bumps something, try to do the Bump Melee Attack.
  (bind _data (Utilities|FlowControl|MergeExecutionwithData (stale Custom|OnActorBeginOverlap_Event OtherActor) OtherActor)
    (:Then
      (bind _asfort_pawn (Utilities|Casting|CastToFortPawn _data)) ; Thing we Bumped or Overlapped is a Pawn?
      (bind _fortplayerpawnholdingchicken (Variables|Default|GetFortPlayerPawnHoldingChicken)
        (:then
          (switch Utilities|FlowControl|Switch|SwitchonEFortTeamAffiliation (Team|GetTeamAffiliation _fortplayerpawnholdingchicken _asfort_pawn)
            (:Neutral
              _)
            (:Hostile
              _))
          (switch Utilities|FlowControl|Switch|SwitchonEFortTeamAffiliation (Team|GetTeamAffiliation self _asfort_pawn)
            (:Neutral
              _)
            (:Hostile
              _))
          (CallFunction|AttemptTriggerAbilityAfterPlayerBumpEvent))
        (:else)))))

(event Custom|AttemptTriggerAbilityAfterPlayerBumpEvent
  (bind _tc_heldbumpmeleeability (Variables|Default|GetTC_HeldBumpMeleeAbility))
  (bind _returnvalue (GameplayTags|MakeLiteralGameplayTagContainer _tc_heldbumpmeleeability))
  (bind _self self)
  (bind _aipawnabilitysystemcomponent (Variables|/Script/FortniteGame.FortAIPawn|GetAIPawnAbilitySystemComponent _self))
  ;[1 1 1] Track Player who is holding Chicken
  ;
  ; If Player Bumps something, try to do the Bump Melee Attack.
  (if (Abilities|HasActiveAbilitywithAnyMatchingTag _aipawnabilitysystemcomponent _returnvalue)
    _
    (elif (Abilities|CanActivateAbilitywithMatchingTag _aipawnabilitysystemcomponent _returnvalue)
      (Abilities|TryActivateAbilitiesbyTag (Variables|/Script/FortniteGame.FortAIPawn|GetAIPawnAbilitySystemComponent self) _returnvalue true :!pinned true) ; Activate Melee Bump Attack!
    )))

(event Custom|TestIfInWindSpline
  ;[1 1 1] Test In Windspline and Stop Projectile
  (if (GameplayTags|HasMatchingGameplayTag self "(TagName=\"Athena.Gameplay.Status.Wind.InWindSpline\")")
    (bind _returnvalue (Utilities|Time|SetTimerbyEvent (EventDispatchers|CreateEvent :!event Custom|TryStopGuideProjectile) 0.1)) ; Delay so the projectile can be created and then destroyed properly. Otherwise causes client / server mismatch
    (Variables|Default|SetTryStopGuildeProjectileTimer _returnvalue)
    (else
      (bind _asynctaskproxy (Ability|Tasks|WaitGameplayTagAddToActor self "(TagName=\"Athena.Gameplay.Status.Wind.InWindSpline\")" true)
        (:then
          (Variables|Default|SetAsync%20Action _asynctaskproxy))
        (:Added
          (CallFunction|TryStopGuideProjectile))))))

(event OnHeldObjectGuideProjectileStarted(FortHeldObject)
  ;[1 1 1] Remove stun, return collision when chicken is thrown or dropped
  (Variables|Default|SetbPawnThrown true)
  (CallFunction|UpdateMovementComponents)
  (Networking|ForceNetUpdate)
  (CallFunction|PlayThrowAudio)
  (CallFunction|TestifinWindSpline))

(event OnHeldObjectGuideProjectileStopped(FortHeldObject)
  (bind _async_action (Variables|Default|GetAsync%20Action))
  (bind _trystopguildeprojectiletimer (Variables|Default|GetTryStopGuildeProjectileTimer))
  ;[1 1 1] Remove stun, return collision when chicken is thrown or dropped
  (Variables|Default|SetbPawnThrown false)
  (CallFunction|UpdateMovementComponents)
  (Networking|ForceNetUpdate)
  (CallFunction|ResetChickenVars)
  (Utilities|IsValid _async_action
    (:Is%20Valid
      (AsyncAction|Cancel _async_action))
    (:Is%20Not%20Valid
      _))
  (if (Utilities|Time|IsValidTimerHandle _trystopguildeprojectiletimer)
    (Utilities|Time|ClearandInvalidateTimerbyHandle _trystopguildeprojectiletimer)))

(event Custom|ForceLanding
  (bind _async_action (Variables|Default|GetAsync%20Action))
  (bind _trystopguildeprojectiletimer (Variables|Default|GetTryStopGuildeProjectileTimer))
  ;[1 1 1] Remove stun, return collision when chicken is thrown or dropped
  (Variables|Default|SetbPawnThrown false)
  (CallFunction|UpdateMovementComponents)
  (Networking|ForceNetUpdate)
  (CallFunction|ResetChickenVars)
  (Utilities|IsValid _async_action
    (:Is%20Valid
      (AsyncAction|Cancel _async_action))
    (:Is%20Not%20Valid
      _))
  (if (Utilities|Time|IsValidTimerHandle _trystopguildeprojectiletimer)
    (Utilities|Time|ClearandInvalidateTimerbyHandle _trystopguildeprojectiletimer)))

(event Custom|TryStopGuideProjectile
  (bind _fortheldobject (Variables|Default|GetFortHeldObject))
  (bind _guideprojectile (Variables|/Script/FortniteGame.FortHeldObjectComponent|GetGuideProjectile _fortheldobject))
  (bind _projectilemovementcomponent (Variables|/Script/FortniteGame.FortProjectileBase|GetProjectileMovementComponent _guideprojectile))
  ;[1 1 1] Test In Windspline and Stop Projectile
  (Utilities|IsValid _fortheldobject
    (:Is%20Valid
      (Utilities|IsValid _guideprojectile
        (:Is%20Valid
          (Utilities|IsValid _projectilemovementcomponent
            (:Is%20Valid
              (Game|Components|ProjectileMovement|StopSimulating _projectilemovementcomponent (Collision|MakeHitResult false false 0.0 0.0 (Transformation|GetActorLocation self)))
              (CallFunction|ForceLanding))))))))

(event OnHeldObjectThrown(FortHeldObject)
  ;[1 1 1] Handle player events related to throwing chickens
  (CallFunction|HandleonHeldObjectThrown))

(event Custom|HandleOnHeldObjectThrown
  ;[1 1 1] Handle player events related to throwing chickens
  (bind _fortplayerpawnholdingchicken (Variables|Default|GetFortPlayerPawnHoldingChicken)
    (:then
      (Quest|SendThrownbyPlayerEvents (Utilities|Set|ContainsItem (Variables|Quest|GetThrownByPlayerIDs) (Variables|/Script/FortniteGame.FortPawn|GetPawnUniqueID _fortplayerpawnholdingchicken)))
      (bind _returnvalue (Utilities|Time|SetTimerbyEvent (delegate Custom|ClearPlayer) 1.0))
      (Variables|Default|SetThrown%20Handle _returnvalue))
    (:else)))

(event Custom|ClearPlayer
  ;[1 1 1] Handle player events related to throwing chickens
  (Variables|Default|SetFortPlayerPawnHoldingChicken 0) ; Clear this variable since the Chicken is no longer held.
)

(event Custom|ResetChickenVars
  ;[1 1 1] End PickedUp stun, return collision vars
  (Variables|Default|SetbPawnPickedUp false)
  (CallFunction|UpdateMovementComponents)
  (Utilities|FlowControl|Delay (Variables|Default|GetPostLandDelay))
  (GameplayEffects|RemoveActiveGameplayEffect (Variables|GameplayAbilities|GetAIPawnAbilitySystemComponent) (Variables|Default|GetH_GE_PickedUp_Chicken) -1)
  (CallFunction|NPCNugUpdateStormHandling)
  (Networking|ForceNetUpdate)
  (if (Networking|IsServer)
    (Utilities|Time|SetTimerbyEvent (delegate Custom|EnablePickUp) (Variables|Default|GetPickupEnableDelay))))

(event Custom|EnablePickUp
  ;[1 1 1] End PickedUp stun, return collision vars
  (bind _output_get (Variables|NPCInteraction|SetNPC%20PawnCanInteractInternal true))
  (AI|SetCanInteract _output_get)
  (Networking|ForceNetUpdate))

(event Custom|PlayThrowAudio
  ;[1 1 1] Sound
  (Default|PlaySoundLib "(TagName=\"SoundLibrary.Irwin.VO.Thrown\")"))

(event Custom|PlayPickUpAudio
  ;[1 1 1] Sound
  (Default|PlaySoundLib "(TagName=\"SoundLibrary.Irwin.VO.PickUp\")"))

(event Custom|NPCNugUpdateStormHandling
  ;[1 1 1] Storm Handling for Nug (when held / dropped again)
  (Networking|SwitchHasAuthority
    (:Authority
      (if (Variables|Default|GetbPawnPickedUp)
        (CallFunction|StopTestingIsinSafeZone) ; If held stop testing in safe zone (clears timers and any pending despawn)
        (else
          (CallFunction|InitializeTestinSafeZone) ; Once dropped start up testing once again
        )))))

(event Custom|NugLanded
  (params (Hit :type Hit%20Result%20Structure%20%28by%20ref%29))
  ;[1 1 1] Landing sounds
  (CallFunction|OnNugLanded Hit))

(event Custom|OnNugLanded
  (params (Hit :type Hit%20Result%20Structure))
  ;[1 1 1] Landing sounds
  (Networking|CanExecuteCosmeticEvents
    (:True
      (Audio|PlaySoundatLocation (Variables|Default|GetNugJumpLandSound) (Transformation|GetActorLocation)))))

(fn BindPlayerControllerOnUnpossess
  (bind _fortheldobject (Variables|Default|GetFortHeldObject))
  (bind _returnvalue (HeldObjects|Instigator|GetInstigator _fortheldobject))
  (Networking|SwitchHasAuthority
    (:Authority
      (Utilities|IsValid _returnvalue
        (:Is%20Valid
          (bind _asfort_player_pawn_athena (Utilities|Casting|CastToFortPlayerPawnAthena _returnvalue))
          (Game|Player|BindEventtoOnPawnUnpossessed _asfort_player_pawn_athena (EventDispatchers|CreateEvent :!event Server%20Only%20Handle%20Player%20Controller%20OnUnPossess)))))))

(fn BlueprintCanInteract
  (params (InteractingPawn :type Fort%20Pawn%20Object%20Reference) (InteractionBeingAttempted :type EInteractionBeingAttempted%20Enum :default "FirstInteraction"))
  (return (Math|Boolean|ANDBoolean (Variables|Default|GetbIsPickupEnabled) (|CallParentFunction InteractingPawn InteractionBeingAttempted) (not (GameplayTags|HasAnyMatchingGameplayTags (Utilities|Casting|CastToFortPlayerPawn InteractingPawn) (Variables|Default|GetTC_BlockPickup))) (not (GameplayTags|HasAnyMatchingGameplayTags (Variables|Default|GetInteraction%20Preventing%20Tags))))))

(fn GetCurrentLocation
  (return "0, 0, 0"))

(fn GetFortProjectileMovementComponent
  (return 0))

(fn GetInstigatorPlayerController
  (return 0))

(fn GetMaxTravelDistanceSquared
  (return 0.0))

(fn IsProjectileBeingKilled
  (return false))

(fn OnLaunchCharacter
  (params (LaunchVelocity :type Vector :default "0, 0, 0") (bXYOverride :type Boolean) (bZOverride :type Boolean))
  (bind _charactermovement (Variables|Character|GetCharacterMovement))
  (bind _returnvalue (|CallParentFunction LaunchVelocity bXYOverride bZOverride))
  (switch Utilities|FlowControl|Switch|SwitchonEMovementMode (Variables|/Script/Engine.CharacterMovementComponent|GetMovementMode _charactermovement)
    (:MOVE_None
      (if (and (Variables|Default|GetbPawnThrown) (ScalableFloat|GetValueasBool (Variables|Default|GetForceMovementModeIfNoMovementModeWhenLaunched)))
        (Pawn|Components|CharacterMovement|SetMovementMode _charactermovement "MOVE_Walking" 0)
        (else
          _))
      (return _returnvalue))
    (:MOVE_Walking
      _)
    (:MOVE_NavWalking
      _)
    (:MOVE_Falling
      _)
    (:MOVE_Swimming
      _)
    (:MOVE_Flying
      _)
    (:MOVE_Custom
      _)))

(fn OnRep_bPawnPickedUp
  (CallFunction|UpdateMovementComponents)
  (if (Variables|Default|GetbPawnPickedUp)
    (CallFunction|AssignPlayerBumpEvent) ; Track whether Pawn that grabbed the Chicken has Bumped something.
    (else
      (CallFunction|UnassignPlayerBumpEvent) ; Stop tracking whether Pawn that grabbed the Chicken has Bumped something.
    )))

(fn OnRep_bPawnThrown
  (CallFunction|UpdateMovementComponents)
  (if (Variables|Default|GetbPawnThrown)
    (CallFunction|UnassignPlayerBumpEvent) ; Stop tracking whether Pawn that grabbed the Chicken has Bumped something.
  ))

(fn SendThrownByPlayerEvents
  (params (bAlreadyThrown :type Boolean))
  (bind _fortplayerpawnholdingchicken (Variables|Default|GetFortPlayerPawnHoldingChicken))
  (bind _returnvalue (Fortnite|GetFortPlayerControllerfromActor _fortplayerpawnholdingchicken))
  (bind _returnvalue_2 (Quests|GetQuestManager _returnvalue "Athena"))
  (if (not bAlreadyThrown)
    ;[1 1 1] Send a throw quest update when a player has thrown this specific chicken for the first time
    (Utilities|Set|Add (Variables|Quest|GetThrownByPlayerIDs) (Variables|/Script/FortniteGame.FortPawn|GetPawnUniqueID (Variables|Default|GetFortPlayerPawnHoldingChicken)))
    (Utilities|IsValid _returnvalue
      (:Is%20Valid
        (Utilities|IsValid _returnvalue_2
          (:Is%20Valid
            (Quest|SendCustomStatEventwithTags _returnvalue_2 "Throw" "(GameplayTags=((TagName=\"Athena.Quests.Chicken.Throw\")))" "(GameplayTags=((TagName=\"Pawn.Athena.NPC.Wildlife.Prey.Nug\")))" 1)))))))

(fn ServerOnlyHandlePlayerControllerOnUnPossess
  (params (PlayerPawn :type Pawn%20Object%20Reference)))

(fn SetPawnMovementEnabled
  (params (bEnabled :type Boolean))
  (Pawn|Components|CharacterMovement|SetMovementMode (Variables|Character|GetCharacterMovement) (select bEnabled "MOVE_Falling" "MOVE_None") 0))

(fn ShouldClampTravelDistance
  (return false))

(fn UpdateCollision
  (bind _outactors (Actor|GetAttachedActors true))
  (bind _capsulecomponent (Variables|Character|GetCapsuleComponent))
  (if (or (Variables|Default|GetbPawnPickedUp) (Variables|Default|GetbPawnThrown))
    (Collision|SetCollisionResponsetoChannel _capsulecomponent "ECC_Pawn" "ECR_Ignore") ; Adjust collision so we don't prevent pawns from riding in the vehicle
    ;[1 1 1] Enable/Disable collision on attached actors (Parasite for example) during hold and throw
    (for _array_element _outactors
      (Collision|SetActorEnableCollision _array_element false))
    (else
      (Collision|SetCollisionResponsetoChannel _capsulecomponent "ECC_Pawn" "ECR_Block") ; Adjust collision so we don't prevent pawns from riding in the vehicle
      (for _array_element_2 _outactors
        (Collision|SetActorEnableCollision _array_element_2 true)))))

(fn UpdateMovementComponents
  (CallFunction|UpdateCollision)
  (if (not (GameplayTags|HasMatchingGameplayTag self "(TagName=\"Athena.Gameplay.Status.Wind.InWindSpline\")"))
    (CallFunction|SetPawnMovementEnabled (not (Math|Boolean|ORBoolean (Variables|Default|GetbPawnPickedUp) (Variables|Default|GetbPawnThrown) false))))
  (bind _returnvalue (Class|FortAnimInputEventBool|CreateAnimInputEventBool "aie.fauna.PawnPickedUp" (Variables|Default|GetbPawnPickedUp)))
  (Animation|TriggerAnimInputEvent _returnvalue)
  (bind _returnvalue_2 (Class|FortAnimInputEventBool|CreateAnimInputEventBool "aie.fauna.PawnThrown" (Variables|Default|GetbPawnThrown)))
  (Animation|TriggerAnimInputEvent _returnvalue_2))

(fn ConstructionScript
  (|CallParentFunction))
